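# Niagara trail effect for the Fuchsia F emote, combining a ribbon trail with electric sprite sparks tinted by User.VariantColor.
# FN_AnimTrail_V2 spawns short-lived ribbon particles at a steady rate with a brightened color and alpha fade.
# GPU emitter lightning001 spawns sprites per distance moved, with inherited and point-based velocity, drag, hot color-curve flash, shrink and random-start SubUV animation.
# A similar second lightning emitter is present but disabled.
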NiagaraSystem /BRCosmetics/Animation/Game/MainPlayer/Emotes/Fuschia/VFX/NS_Emote_Fuchsia_F_Trail
dependencies {
 usedDataInterfaces [&NiagaraDataInterfaceCurve, &NiagaraDataInterfaceColorCurve]
 usedDynamicInputs []
 usedModules [&SystemState]
}
emitters {
 FN_AnimTrail_V2 {
  emitterUpdateScript [
   &EmitterScalability(
    Module.bApplySpawnCountScale=false
    Module.bApplySpawnCountScaleByDistanceFraction=false
    Module.bUseMaxDistance=false
    Module.bUseMinDistance=false
   ) {enabled=false}
   &EmitterLifeCycle(NextLoopDuration=0.8917) {enabled=false}
   &EmitterState(Life Cycle Mode=::NewEnumerator1, Loop Behavior=::NewEnumerator1, Loop Duration=6, UseLoopDelay=false)
   &SpawnRate(SpawnRate=20, Use Spawn Probability=false)
  ]
  particleSpawnScript [/Game/Effects/Niagara/Modules/Contrails/FN_InitializeRibbon/FN_InitializeRibbon(
    Color:LinearColor=$User.VariantColor
    Facing Axis:ENiagaraOrientationAxis=::ZAxis
    Lifetime=0.35
    Offset Amount=50
    Offset Axis:ENiagaraOrientationAxis=::ZAxis
    Ribbon Width=70
   )]
  particleUpdateScript [
   &SolveForcesAndVelocity(Clamp Velocity=false, Limit Acceleration=false)
   &FN_UpdateRibbon(Initial Color:LinearColor=$Particles.Initial.Color, Initial RibbonWidth:NiagaraFloat=$Particles.Initial.RibbonWidth)
   &ScaleColor(
    Linear Color Curve=DataInterface() {propertyValues {
      AlphaCurve {keys [<at 0 0>, <at 0.0304 0.7 RCIM_Cubic>, {interpMode=RCIM_Cubic, leaveTangent=-0.8431, time=0.973, value=0}]}
      BlueCurve {keys [<at 0.0114 3>]}
      GreenCurve {keys [<at 0.0114 3>]}
      RedCurve {keys [<at 0.0114 3>]}
     }}
    Scale Mode=::NewEnumerator2
   )
  ]
  rendererClasses [&NiagaraRibbonRendererProperties]
  renderers [{rendererClass=&NiagaraRibbonRendererProperties}]
 }
 lightning {
  bEnabled=false
  simTarget=GPUComputeSim
  emitterSpawnScript [&FN_PreserveSkydiveUserVariables() {module=PreserveUserVariables}]
  emitterUpdateScript [
   &EmitterState(Life Cycle Mode=::NewEnumerator1, Loop Behavior=::NewEnumerator1, Loop Delay=-0.0121, Loop Duration=6, UseLoopDelay=true)
   &SpawnPerUnit(Module.Use Max Movement Threshold=true, Module.Use Movement Tolerance=false, Spawn Per Unit=44, Use Spawn Probability=false)
   &SpawnRate(SpawnRate=20, Use Spawn Probability=false) {enabled=false}
  ]
  particleSpawnScript [
   &InitializeParticle(
    Color=&MultiplyLinearColor(Linear Color=$User.VariantColor, Scale Factor=<5 5 5 0.8>)
    Lifetime=&UniformRangedFloat(Minimum=0.1, Maximum=0.35)
    Sprite Rotation Mode=::NewEnumerator1
    Sprite Size Mode=::NewEnumerator1
    Sprite UV Mode=::NewEnumerator3
    Uniform Sprite Size Max=33
    Uniform Sprite Size Min=20
    UsePositionOffset=false
    Write Material Random=true
   ) {module=InitializeParticle001}
   &BoxLocation(Box Offset=<0 0 20>, Box Size=<7 1 30>, MaskSpawnGroup=false, Module.Override Seed=false, Rotate Box=false, Surface Only=false)
   &AddVelocity(Velocity=&UniformRangedVector(Minimum=<0 0 0>, Maximum=<0 2 0>))
   &InheritVelocity(
    Module.Limit Speed=true
    Velocity Limit=10000
    Velocity Scale=&RandomRangeVector(Minimum=<0 0 0>, Maximum=<0.2 0.2 0.2>)
   ) {module=InheritVelocity001}
   &AddVelocityFromPoint(
    Module.Use Falloff Distance=false
    Velocity Origin:Vector3f=&SimulationPosition()
    Velocity Strength=&UniformRangedFloat(Minimum=2, Maximum=100)
   )
   /SetVariables(Particles.RandomFrame=&RandomRangeInt(Minimum=0, Maximum=63)) {isSetParameters=true, module=SetVariables_#1}
  ]
  particleUpdateScript [
   &ParticleState()
   &GravityForce(Gravity=<0 0 -200>) {enabled=false}
   &Drag(Drag=2, Use Linear Drag=true, Use Rotational Drag=false)
   &SolveForcesAndVelocity(Clamp Velocity=false, Limit Acceleration=false)
   &ScaleColor(
    Scale RGB=&MakeVectorFromLinearColorRGB(Color=&ColorFromCurve(ColorCurve=DataInterface() {propertyValues {
        AlphaCurve {keys [<at 0 1 RCIM_Cubic>,<at 1 1 RCIM_Cubic>]}
        BlueCurve {keys [<at 0.2 16.8947 RCIM_Cubic>,<at 0.7 1 RCIM_Cubic>]}
        GreenCurve {keys [<at 0.2 11.9097 RCIM_Cubic>,<at 0.7 0.7365 RCIM_Cubic>]}
        RedCurve {keys [<at 0.2 20 RCIM_Cubic>,<at 0.7 0.5291 RCIM_Cubic>]}
       }}))
    ScaleA=true
    ScaleRGB=true
   ) {module=ScaleColor001}
   &ScaleSpriteSize(Scale Factor=&Vector2DFromFloat(Value=&FloatFromCurve(FloatCurve=DataInterface() {propertyValues {Curve {keys [{arriveTangent=-6.725, interpMode=RCIM_Cubic, leaveTangent=-6.725, tangentMode=RCTM_User, time=0.65, value=1}, <at 1 0 RCIM_Cubic>]}}})))
   &ScaleSpriteSizeBySpeed(Max Scale Factor=<5 10>, Min Scale Factor=<5 0.5>)
   &FN_SubUVAnimation(End Frame=&Add_Integer(A=$Particles.RandomFrame, B=24), Start Frame=$Particles.RandomFrame, SubUV Loop Count=2)
  ]
  rendererClasses [&NiagaraSpriteRendererProperties]
  renderers [{rendererClass=&NiagaraSpriteRendererProperties}]
 }
 lightning001 {
  simTarget=GPUComputeSim
  emitterSpawnScript [&FN_PreserveSkydiveUserVariables() {module=PreserveUserVariables}]
  emitterUpdateScript [
   &EmitterState(Life Cycle Mode=::NewEnumerator1, Loop Behavior=::NewEnumerator1, Loop Delay=-0.0121, Loop Duration=6, UseLoopDelay=false)
   &SpawnPerUnit(
    Max Movement Threshold=2000
    Module.Use Max Movement Threshold=true
    Module.Use Movement Tolerance=true
    Movement Tolerance=0.05
    Spawn Per Unit=10
    Use Spawn Probability=false
   )
   &SpawnRate(SpawnRate=20, Use Spawn Probability=false) {enabled=false}
  ]
  particleSpawnScript [
   &InitializeParticle(
    Color=&MultiplyLinearColor(Linear Color=$User.VariantColor, Scale Factor=<4 5 5 0.8>)
    Lifetime=&UniformRangedFloat(Minimum=0.1, Maximum=0.2)
    Sprite Rotation Mode=::NewEnumerator1
    Sprite Size Mode=::NewEnumerator1
    Sprite UV Mode=::NewEnumerator3
    Uniform Sprite Size Max=35
    Uniform Sprite Size Min=30
    UsePositionOffset=false
    Write Material Random=true
   ) {module=InitializeParticle001}
   &BoxLocation(Box Offset=<0 0 30>, Box Size=<1 1 70>, MaskSpawnGroup=false, Module.Override Seed=false, Rotate Box=false, Surface Only=false)
   &AddVelocity(Velocity=&UniformRangedVector(Minimum=<0 0 0>, Maximum=<0 2 0>)) {enabled=false}
   &InheritVelocity(
    Module.Limit Speed=true
    Velocity Limit=10000
    Velocity Scale=&RandomRangeVector(Minimum=<0 0 0>, Maximum=<-0.2 -0.2 -0.2>)
   ) {module=InheritVelocity001}
   &AddVelocityFromPoint(
    Module.Use Falloff Distance=false
    Velocity Origin:Vector3f=&SimulationPosition()
    Velocity Strength=&UniformRangedFloat(Minimum=2, Maximum=100)
   )
   &SpriteFacingAndAlignment(Facing Coordinate Space=::Local, WriteAlignment=false, WriteFacing=true)
   /SetVariables(Particles.RandomFrame=&RandomRangeInt(Minimum=0, Maximum=63)) {isSetParameters=true, module=SetVariables_#1}
   &CameraOffset(Camera Offset Amount=-1)
  ]
  particleUpdateScript [
   &ParticleState()
   &GravityForce(Gravity=<0 0 -200>) {enabled=false}
   &Drag(Drag=20, Use Linear Drag=true, Use Rotational Drag=false)
   &SolveForcesAndVelocity(Clamp Velocity=false, Limit Acceleration=false)
   &ScaleColor(
    Scale RGB=&MakeVectorFromLinearColorRGB(Color=&ColorFromCurve(ColorCurve=DataInterface() {propertyValues {
        AlphaCurve {keys [<at 0 0 RCIM_Cubic>,<at 0.0336 0>,<at 0.0698 1>,<at 1 1 RCIM_Cubic>]}
        BlueCurve {keys [<at 0.2 16.8947 RCIM_Cubic>,<at 0.7 1 RCIM_Cubic>]}
        GreenCurve {keys [<at 0.2 11.9097 RCIM_Cubic>,<at 0.7 0.7365 RCIM_Cubic>]}
        RedCurve {keys [<at 0.2 20 RCIM_Cubic>,<at 0.7 0.5291 RCIM_Cubic>]}
       }}))
    ScaleA=true
    ScaleRGB=true
   ) {module=ScaleColor001}
   &ScaleSpriteSize(Scale Factor=&Vector2DFromFloat(Value=&FloatFromCurve(FloatCurve=DataInterface() {propertyValues {Curve {keys [{arriveTangent=-6.725, interpMode=RCIM_Cubic, leaveTangent=-6.725, tangentMode=RCTM_User, time=0.65, value=1}, <at 1 0 RCIM_Cubic>]}}})))
   &ScaleSpriteSizeBySpeed(Max Scale Factor=<4 4>, Min Scale Factor=<3 3>)
   &FN_SubUVAnimation(End Frame=&Add_Integer(A=$Particles.RandomFrame, B=24), Start Frame=$Particles.RandomFrame, SubUV Loop Count=2)
  ]
  rendererClasses [&NiagaraSpriteRendererProperties]
  renderers [{rendererClass=&NiagaraSpriteRendererProperties}]
 }
}
summary {
 systemName=NS_Emote_Fuchsia_F_Trail
 emitters [
  {emitterName=FN_AnimTrail_V2, rendererClasses [&NiagaraRibbonRendererProperties]}
  {emitterName=lightning001, simTarget=GPUComputeSim, rendererClasses [&NiagaraSpriteRendererProperties]}
  {bEnabled=false, emitterName=lightning, simTarget=GPUComputeSim, rendererClasses [&NiagaraSpriteRendererProperties]}
 ]
 userVariables [
  User.bIsFrontEnd {type=NiagaraBool}
  User.bIsFrontEndPreview {type=NiagaraBool}
  User.PlayerSpeed {type=NiagaraFloat, defaultValue=0, description=""}
  User.RibbonIDOffset {type=NiagaraInt32}
  User.SkeletalMesh {
   type=NiagaraDataInterfaceSkeletalMesh
   defaultValue {
    struct=NiagaraExt_VariableValue_DataInterface
    value {dataInterface=&self:NiagaraDataInterfaceSkeletalMesh_0}
   }
  }
  User.VariantColor {type=LinearColor, description="", defaultValue {struct=LinearColor, value=<rgba 0.2038 0.0293 1 1>}}
 ]
}
